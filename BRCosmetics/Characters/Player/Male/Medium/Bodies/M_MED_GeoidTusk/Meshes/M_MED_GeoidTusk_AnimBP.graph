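# Animation Blueprint for the GeoidTusk medium male player body cosmetic, targeting that body's own skeleton.
# It is effectively an empty shell: the event graph holds only an unconnected Blueprint Update Animation event, with no variables.
# No animation graph logic or pose processing appears in it.
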AnimBlueprint /BRCosmetics/Characters/Player/Male/Medium/Bodies/M_MED_GeoidTusk/Meshes/M_MED_GeoidTusk_AnimBP
target_skeleton=&/M_MED_GeoidTusk_Skeleton
graphs [EventGraph {nodes [Event BlueprintUpdateAnimation()]}]
variables []
